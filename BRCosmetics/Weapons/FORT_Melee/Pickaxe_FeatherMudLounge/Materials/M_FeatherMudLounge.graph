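# Surface material for the FeatherMudLounge pickaxe, adding an animated fiery orange emissive glow.
# It wraps the 2023 character parent PreFX and PostFX functions, overriding emissive and tinting base color orange through an FX mask.
# The glow combines a shape texture, panning noise between two HDR oranges, inverted Fresnel and a time-based sine flicker.
# A sphere mask in pre-skinned local space confines the glow to one region of the mesh.
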Material /BRCosmetics/Weapons/FORT_Melee/Pickaxe_FeatherMudLounge/Materials/M_FeatherMudLounge
graph {
 properties {
  bUseMaterialAttributes=true
  bUsedWithMeshDeformer=true
  bUsedWithMorphTargets=true
  bUsedWithSkeletalMesh=true
  pixelDepthOffsetMode=PDOM_Legacy
  shadingModel=MSM_FromMaterialExpression
  usedShadingModels=Default Lit | Subsurface | Clear Coat | Subsurface Profile | Cloth
 }
 nodes [
  n1 = MaterialFunctionCall() {materialFunction=&MF_CharacterParent_2023_PreFX}
  n2 = GetMaterialAttributes(None=@n1.Result) {attributeGetTypes [#1]}
  n3 = Constant3Vector() {constant=<rgba 1 0.4325 0.0534 1>}
  n4 = TextureSample() {
   samplerSource=SSM_Wrap_WorldGroupSettings
   samplerType=SAMPLERTYPE_Masks
   texture=/BRCosmetics/Weapons/FORT_Melee/Pickaxe_FeatherMudLounge/Textures/T_FeatherMudLounge_Pickaxe_FX
  }
  MaterialFunctionCall(
   Material=SetMaterialAttributes!(
    MaterialAttributes=@n2.MaterialAttributes
    Emissive Color=NamedRerouteUsage()
    Base Color=LinearInterpolate(A=@n2.Base Color, B=@n3, Alpha=@n4.G)
   ) {attributeSetTypes [#2, #1]}
   M=@n1.M
   Pre Skinned Local Position=@n1.M
   Distance Blend=@n1.M
   Fresnel=@n1.M
  ) {materialFunction=&MF_CharacterParent_2023_PostFX}
  n5 = Multiply(A=Time()) {constB=0.8}
  NamedRerouteDeclaration(None=Saturate(None=Sine(None=@n5))) {name=Flicker, nodeColor=<rgba 1 0 0.1406 1>}
  NamedRerouteDeclaration(None=@n1.Pre skinned Local Position) {name=PreSkinnedLocalPosition, nodeColor=<rgba 0 0.4609 1 1>}
  n6 = Power(Base=TextureSample() {texture=/BRCosmetics/Weapons/FORT_Melee/Pickaxe_FeatherMudLounge/Textures/T_FeatherMudLounge_Axe1Shape_E2}.R) {constExponent=1.1}
  n7 = TextureCoordinate()
  n8 = TextureSample(UVs=Panner(Coordinate=Multiply(A=@n7)) {speedX=0.2, speedY=0.2}) {samplerSource=SSM_Wrap_WorldGroupSettings, samplerType=SAMPLERTYPE_Grayscale, texture=&LowResBlurredNoise}
  n9 = LinearInterpolate(A=Constant3Vector() {constant=<rgba 50 8.4684 0 1>}, B=Constant3Vector() {constant=<rgba 50 21.7689 0 1>}, Alpha=@n8.R)
  n10 = NamedRerouteUsage()
  n11 = Fresnel() {exponent=0.2}
  n12 = NamedRerouteUsage()
  n13 = Multiply(
   A=Power(Base=OneMinus(None=@n11), Exp=LinearInterpolate(Alpha=@n12) {constA=3}) {constExponent=1}
   B=LinearInterpolate(Alpha=NamedRerouteUsage()) {constA=1, constB=2}
  ) {constB=2}
  n14 = Multiply(
   A=Multiply(A=Power(Base=@n8.R) {constExponent=1}, B=Multiply(A=@n9) {constB=0.7})
   B=LinearInterpolate(A=LinearInterpolate(Alpha=@n10) {constA=0.25, constB=0.6}, Alpha=Saturate(None=@n13))
  )
  n15 = SphereMask(
   A=NamedRerouteUsage()
   B=Constant3Vector() {constant=<rgba -12.2751 0 37.0574 1>}
   Radius=Constant() {r=42.2926}
   Hardness=Constant() {r=0.8}
  ) {hardnessPercent=0.8}
  NamedRerouteDeclaration(None=Multiply(A=Multiply(A=@n6, B=@n14), B=Saturate(None=@n15))) {name=Emissive out, nodeColor=<rgba 0.7812 1 0 1>}
 ]
}
