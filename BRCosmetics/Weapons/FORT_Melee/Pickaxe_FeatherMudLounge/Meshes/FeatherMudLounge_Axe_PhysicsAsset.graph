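# Physics asset for the FeatherMudLounge pickaxe, a two-handed melee axe cosmetic in Battle Royale.
# It defines a single capsule body on the Melee_TwoHanded_Handle bone, offset along the handle and rotated to lie along it.
# It has no constraints, so it provides only simple collision bounds with no jointed simulation between parts.
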PhysicsAsset /BRCosmetics/Weapons/FORT_Melee/Pickaxe_FeatherMudLounge/Meshes/FeatherMudLounge_Axe_PhysicsAsset
bodies {Melee_TwoHanded_Handle [{shapeType=Capsule, length=23.0792, radius=6.5668, center=<0.3043 -0.402 13.7021>, rotation=<rot -0.0367 -90 0.1529>}]}
constraints []
